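# Lit material for the Creative Round Settings device screen, showing glossy green scrolling code with scanlines.
# A code texture scrolls in irregular jumps, driven by two summed stepped timers, and adds to a faint circular mask for tinted emissive.
# Repeating V-axis scanlines multiply base color and emissive, modulated by a synced distance fade at actor position.
# The resulting attributes pass through MF_CreativeMoveTool before output.
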Material /Game/Creative/Devices/RoundSettings/Materials/M_Round_Settings_Screen
graph {
 properties {bUseMaterialAttributes=true, pixelDepthOffsetMode=PDOM_Legacy, usedShadingModels=Default Lit}
 nodes [
  n1 = Multiply(A=ComponentMask(None=TextureCoordinate()) {g=true}, B=ScalarParameter() {defaultValue=80, parameterName=ScanlineRepeat})
  n2 = LinearInterpolate(
   A=SphereMask(
    A=Frac(None=@n1)
    B=Constant() {r=0.5}
    Radius=ScalarParameter() {defaultValue=0.5, parameterName=ScanlineRadius}
    Hardness=ScalarParameter() {defaultValue=0.2, parameterName=ScanlineHarndess}
   )
   Alpha=MaterialFunctionCall(
    Position=ActorPositionWS()
    Fade_Start_Distance=ScalarParameter() {defaultValue=700, parameterName=Fade_Start_Distance}
    Fade_End_Distance=ScalarParameter() {defaultValue=200, parameterName=Fade_End_Distance}
   ) {materialFunction=&MF_Distance_SyncedFade_FX}.Result
  )
  n3 = TextureCoordinate()
  n4 = Constant() {r=0.5}
  n5 = Constant()
  n6 = TextureCoordinate() {vTiling=0.25}
  n7 = Time() {bOverride_Period=true, period=100}
  n8 = ScalarParameter() {defaultValue=0.62, parameterName=TimeScaleA}
  n9 = ScalarParameter() {defaultValue=2.233, parameterName=TimeScaleB}
  n10 = Add(A=Ceil(None=Multiply(A=@n7, B=@n8)), B=Ceil(None=Multiply(A=@n7, B=@n9)))
  n11 = Add(
   A=Multiply(A=SphereMask(A=@n3, B=@n4, Radius=@n4, Hardness=@n5), B=ScalarParameter() {defaultValue=0.03, parameterName=GradientScale})
   B=TextureSample(UVs=Panner(Coordinate=@n6, Time=@n10) {bFractionalPart=true, speedY=0.02}) {texture=/Game/Creative/Devices/RoundSettings/Textures/T_Round_Settings_ScreenCode}.R
  )
  n12 = VectorParameter() {parameterName=TextColor, defaultValue=<rgba 2.0828 4 1.1597 0>}
  n13 = SetMaterialAttributes!(
   MaterialAttributes
   Base Color=Multiply(A=VectorParameter() {parameterName=BaseColor, defaultValue=<rgba 0.0065 0.2205 0.041 0>}.RGB, B=@n2)
   Metallic=ScalarParameter() {parameterName=Metallic}
   Specular=ScalarParameter() {defaultValue=0.8, parameterName=Specular}
   Roughness=ScalarParameter() {defaultValue=0.1, parameterName=Roughness}
   Emissive Color=Multiply(A=Multiply(A=@n11, B=@n12.RGB), B=@n2)
  ) {attributeSetTypes [
    #1
    #2
    #3
    #4
    #5
   ]}
  MaterialFunctionCall(MA In=@n13) {materialFunction=&MF_CreativeMoveTool}
 ]
}
